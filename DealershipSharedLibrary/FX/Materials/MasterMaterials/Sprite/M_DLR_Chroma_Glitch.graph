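# Translucent unlit master material for Niagara sprites and mesh particles that renders a texture with animated chromatic glitching.
# A screen-space glitch texture and panning pixel-grid masks distort the base UVs and add shifting RGB color, tuned through particle dynamic parameters.
# Time-stepped masks cycle the base texture's channels for opacity, while IsGlow selects a single channel instead.
# An optional procedural CRT pattern modulates emissive, and depth fade softens alpha except on low quality.
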Material /DealershipSharedLibrary/FX/Materials/MasterMaterials/Sprite/M_DLR_Chroma_Glitch
graph {
 properties {
  bAutomaticallySetUsageInEditor=false
  bUsedWithNiagaraMeshParticles=true
  bUsedWithNiagaraSprites=true
  blendMode=BLEND_Translucent
  shadingModel=MSM_Unlit
  translucencyLightingMode=TLM_VolumetricPerVertexDirectional
  usedShadingModels=Unlit
 }
 nodes [
  n1 = TextureCoordinate()
  n2 = NamedRerouteUsage()
  n3 = Multiply(A=Add(A=Multiply(B=@n2) {constA=0.5}, B=NamedRerouteUsage()), B=NamedRerouteUsage())
  n4 = ScreenPosition()
  n5 = NamedRerouteDeclaration(None=Multiply(A=NamedRerouteUsage(), B=ComponentMask(None=@n4.ViewportUV) {g=true, r=true})) {name=UV_ScreenPos, nodeColor=<rgba 1 0 0.6094 1>}
  n6 = TextureSampleParameter2D(UVs=@n5) {parameterName=GlitchTexture, texture=/DealershipSharedLibrary/FX/Textures/Digital/T_DLR_Glitch_FB_RGB_03_T}
  n7 = NamedRerouteUsage()
  n8 = ScalarParameter() {defaultValue=0.15, parameterName=Glitch - Speed Intensity}
  n9 = Time()
  n10 = Sine(None=@n9) {period=8}
  n11 = TextureSampleParameter2D(UVs=LinearInterpolate(A=@n5, B=Add(A=@n5, B=@n7), Alpha=Multiply(A=@n8, B=@n10))) {parameterName=GlitchTexture, texture=&T_DLR_Glitch_FB_RGB_03_T}
  n12 = AppendVector(A=Desaturation(None=@n6.RGB), B=@n11.G)
  n13 = AppendVector(A=@n11.B, B=@n6.R)
  n14 = Add(A=@n12, B=@n13)
  n15 = TextureSampleParameter2D(UVs=Add(A=Add(A=@n1, B=@n3), B=Multiply(A=@n14, B=@n2))) {parameterName=Base Texture, texture=/DealershipSharedLibrary/Assets/Sets/FullscreenGlitch/T_DLR_Triangle_RGB_T}
  n16 = ChannelMaskParameter(None=@n15.RGBA) {maskChannel=Green, parameterName=GlowChannelSelect, defaultValue=<rgba 0 1 0 0>}
  n17 = LinearInterpolate(A=@n15.RGB, B=@n16.R, Alpha=ScalarParameter() {parameterName=IsGlow})
  n18 = AppendVector(A=@n12, B=@n11.R)
  n19 = AppendVector(A=@n13, B=@n11.G)
  n20 = DynamicParameter() {parameterIndex=1, paramNames [GlitchTexture - Color Vibrance, Param2, Param3, Param4]}
  n21 = Add(
   A=@n17
   B=NamedRerouteDeclaration(None=LinearInterpolate(A=@n18, B=@n19, Alpha=@n20.GlitchTexture - Color Vibrance)) {name=Glitch Color, nodeColor=<rgba 0.2147 0.3092 1 1>}
  )
  n22 = Multiply(A=@n9, B=ScalarParameter() {defaultValue=16, parameterName=ColorShift_Speed})
  n23 = ScalarParameter() {defaultValue=4, parameterName=Rows}
  n24 = Fmod(A=Floor(None=@n22), B=@n23)
  n25 = Saturate(None=OneMinus(None=@n24))
  n26 = Fmod(A=Ceil(None=@n22), B=@n23)
  n27 = Saturate(None=OneMinus(None=@n26))
  n28 = Saturate(None=@n24)
  n29 = Saturate(None=@n26)
  n30 = Add(A=Add(A=Multiply(A=@n15.R, B=@n25), B=Multiply(A=@n15.G, B=@n27)), B=Multiply(A=@n15.B, B=Multiply(A=@n28, B=@n29)))
  n31 = Multiply(
   A=LinearInterpolate(B=NamedRerouteUsage(), Alpha=ScalarParameter() {defaultValue=1, parameterName=IsCRT}) {constA=1}
   B=LinearInterpolate(A=LinearInterpolate(A=@n21, B=@n17, Alpha=@n30), B=@n17, Alpha=ScalarParameter() {parameterName=IsGlow})
  )
  n32 = ParticleColor()
  MaterialFunctionCall(Emissive=Multiply(A=Multiply(A=@n31, B=@n32.RGB), B=ScalarParameter() {defaultValue=1, parameterName=Global Emissive Gain})) {materialFunction=&MF_FX_EmissiveAdjustments}
  n33 = ScalarParameter() {parameterName=IsGlow}
  Multiply(
   A=Multiply(A=@n32.A, B=LinearInterpolate(A=@n30, B=@n16.R, Alpha=@n33))
   B=QualitySwitch(Default=DepthFade() {fadeDistanceDefault=400}, Low=Constant() {r=1})
  )
  n34 = Constant() {r=892.3058}
  n35 = ScreenPosition()
  n36 = TextureCoordinate()
  n37 = Add(A=Multiply(A=@n36, B=Constant() {r=2}), B=Constant() {r=-1})
  n38 = DotProduct(A=@n37, B=@n37)
  n39 = AppendVector(A=@n38, B=Saturate(None=@n38))
  n40 = Add(A=ComponentMask(None=Multiply(A=@n34, B=@n35.ViewportUV)) {g=true, r=true}, B=@n39)
  n41 = BumpOffset(Coordinate=@n40, Height=Constant() {r=8})
  n42 = Time()
  n43 = Multiply(A=@n42) {constB=6}
  n44 = Fmod(A=Add(A=ComponentMask(None=@n41) {g=true}, B=@n43), B=Constant() {r=0.4})
  n45 = Multiply(A=Constant() {r=2280.2507}, B=@n35.ViewportUV)
  n46 = BumpOffset(Coordinate=Add(A=ComponentMask(None=@n45) {g=true, r=true}, B=@n39), Height=Constant() {r=6})
  n47 = Fmod(A=Add(A=ComponentMask(None=@n46) {g=true}, B=@n43), B=Constant() {r=0.2})
  n48 = Add(A=Abs(None=ConstantBiasScale(None=@n44) {bias=-0.5, scale=2}), B=Abs(None=ConstantBiasScale(None=@n47) {bias=-0.5, scale=2}))
  n49 = Multiply(A=@n42) {constB=0.5}
  n50 = Fmod(A=Add(A=ComponentMask(None=@n41) {r=true}, B=@n49), B=Constant() {r=1})
  n51 = BumpOffset(Coordinate=@n40, Height=Constant() {r=-4})
  n52 = Fmod(A=Add(A=ComponentMask(None=@n51) {g=true}, B=@n49), B=Constant() {r=1})
  n53 = Add(A=Abs(None=ConstantBiasScale(None=@n50) {bias=-0.5, scale=2}), B=Abs(None=ConstantBiasScale(None=@n52) {bias=-0.5, scale=2}))
  NamedRerouteDeclaration(None=Add(A=Multiply(A=@n48, B=@n53), B=@n53)) {name=CRT_Effect, nodeColor=<rgba 0 0.6484 1 1>}
  n54 = DynamicParameter() {defaultValue=<rgba 1 0.35 0 0>, paramNames [GlitchTexture_Tile, GlitchTexture Distortion, PixelGrid Distortion, PixelGrid_PanSpeed]}
  NamedRerouteDeclaration(None=@n54.GlitchTexture_Tile) {name=DP GlitchTexture Tile, nodeColor=<rgba 1 0 0.6654 1>}
  n55 = Constant4Vector() {constant=<rgba 0.125 0.25 -0.15 0.1>}
  n56 = ScalarParameter() {defaultValue=4, parameterName=GlitchUVTile}
  n57 = Multiply(A=Multiply(A=@n36, B=AppendVector(A=@n55.R, B=@n55.R)), B=@n56)
  n58 = Multiply(A=Time(), B=NamedRerouteUsage()) {constB=0}
  n59 = TextureObject() {samplerType=SAMPLERTYPE_LinearColor, texture=/Game/Global/MaterialFunctions/Cosmetics/AnimatedGrid/Examples/Textures/T_AnimatedGridMask}
  n60 = TextureSample(UVs=Panner(Coordinate=Frac(None=@n57), Time=@n58) {speedX=0.1, speedY=-0.5}, Tex=@n59) {samplerType=SAMPLERTYPE_LinearColor, texture=&T_AnimatedGridMask}
  n61 = Multiply(A=@n36, B=AppendVector(A=@n55.B, B=@n55.B))
  n62 = Panner(Coordinate=Frac(None=Multiply(A=@n61, B=@n56)), Time=@n58) {speedX=-0.1, speedY=0.1}
  n63 = AppendVector(A=Multiply(A=@n60.R) {constB=0.4}, B=Multiply(A=TextureSample(UVs=@n62, Tex=@n59).G) {constB=0.4})
  n64 = Constant4Vector() {constant=<rgba 0.1 0.1 0.1 -0.1>}
  n65 = Multiply(A=@n36, B=AppendVector(A=@n64.R, B=@n64.R))
  n66 = Panner(Coordinate=Frac(None=Multiply(A=@n65, B=@n56)), Time=@n58) {speedX=0.25, speedY=0.1}
  n67 = Multiply(A=@n36, B=AppendVector(A=@n64.B, B=@n64.B))
  n68 = Panner(Coordinate=Frac(None=Multiply(A=@n67, B=@n56)), Time=@n58) {speedX=-0.1, speedY=0.1}
  n69 = AppendVector(A=Multiply(A=TextureSample(UVs=@n66, Tex=@n59).R) {constB=0.4}, B=Multiply(A=TextureSample(UVs=@n68, Tex=@n59).G) {constB=0.4})
  NamedRerouteDeclaration(None=Add(A=Multiply(A=@n63) {constB=0.8}, B=Multiply(A=@n69) {constB=0.8})) {name=UV_PixelGrid, nodeColor=<rgba 0.1975 0.0011 1 1>}
  NamedRerouteDeclaration(None=@n54.GlitchTexture Distortion) {name=DP GlitchTexture Distortion, nodeColor=<rgba 1 0 0.6654 1>}
  NamedRerouteDeclaration(None=@n54.PixelGrid Distortion) {name=DP PixelGrid Distortion, nodeColor=<rgba 0.8858 0 1 1>}
  NamedRerouteDeclaration(None=@n54.PixelGrid_PanSpeed) {name=DP PixelGrid Pan Speed, nodeColor=<rgba 1 0 0.3984 1>}
  ScalarParameter() {defaultValue=0.44, parameterName=BaseUVGlitch_PanSpeed}
 ]
}
