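# Unlit, two-sided translucent master material for Niagara sprite and mesh particles, drawing textures in radial UV space.
# Centered radius and arctangent build twistable UVs, offset and tiled by dynamic parameters, sampled with raw-UV derivatives.
# Base and optional secondary sprites combine with a radial or simple-panned mask, then smoothstep dissolve.
# Particle color, optionally texture-tinted, drives emissive; alpha, camera fade and SM5+ deferred depth fade shape opacity.
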Material /DealershipSharedLibrary/FX/Materials/MasterMaterials/Sprite/M_DLR_RadialSprites
graph {
 expressions {
  MaterialExpressionAdd_0 {
   class=MaterialExpressionAdd
   inputs [
    A {from=MaterialExpressionAppendVector_4}
    B {from=MaterialExpressionReroute_4}
   ]
  }
  MaterialExpressionAdd_1 {
   class=MaterialExpressionAdd
   inputs [
    A {from=MaterialExpressionAdd_3}
    B {from=MaterialExpressionAppendVector_3}
   ]
  }
  MaterialExpressionAdd_2 {
   class=MaterialExpressionAdd
   inputs [
    A {from=MaterialExpressionArctangent2Fast_0}
    B {from=MaterialExpressionReroute_8}
   ]
  }
  MaterialExpressionAdd_3 {
   class=MaterialExpressionAdd
   inputs [
    A {from=MaterialExpressionAppendVector_5}
    B {from=MaterialExpressionReroute_4}
   ]
  }
  MaterialExpressionAdd_4 {
   class=MaterialExpressionAdd
   inputs [
    A {from=MaterialExpressionChannelMaskParameter_0, output=R}
    B {from=MaterialExpressionChannelMaskParameter_1, output=R}
   ]
  }
  MaterialExpressionAdd_5 {
   class=MaterialExpressionAdd
   inputs [
    A {from=MaterialExpressionLinearInterpolate_4}
    B {from=MaterialExpressionDivide_0}
   ]
  }
  MaterialExpressionAdd_6 {
   class=MaterialExpressionAdd
   inputs [
    A {from=MaterialExpressionComponentMask_0}
    B {from=MaterialExpressionComponentMask_1}
   ]
  }
  MaterialExpressionAppendVector_1 {
   class=MaterialExpressionAppendVector
   inputs [
    A {from=MaterialExpressionLinearInterpolate_1, output=""}
    B {from=MaterialExpressionLinearInterpolate_2, output=""}
   ]
  }
  MaterialExpressionAppendVector_2 {
   class=MaterialExpressionAppendVector
   inputs [
    A {from=MaterialExpressionReroute_15, output=""}
    B {from=MaterialExpressionAdd_5, output=""}
   ]
  }
  MaterialExpressionAppendVector_3 {
   class=MaterialExpressionAppendVector
   inputs [
    A {from=MaterialExpressionDynamicParameter_6, output=Sprite Extrusion Speed}
    B {from=MaterialExpressionDynamicParameter_6, output=Sprite Extrusion Speed}
   ]
  }
  MaterialExpressionAppendVector_4 {
   class=MaterialExpressionAppendVector
   inputs [
    A {from=MaterialExpressionDynamicParameter_7, output=Mask Offset U}
    B {from=MaterialExpressionDynamicParameter_7, output=Mask Offset U}
   ]
  }
  MaterialExpressionAppendVector_5 {
   class=MaterialExpressionAppendVector
   inputs [
    A {from=MaterialExpressionDynamicParameter_7, output=Sprite Offset U}
    B {from=MaterialExpressionDynamicParameter_7, output=Sprite Offset U}
   ]
  }
  MaterialExpressionAppendVector_6 {
   class=MaterialExpressionAppendVector
   inputs [
    A {from=MaterialExpressionDynamicParameter_3, output=Base Sprite Tile U}
    B {from=MaterialExpressionDynamicParameter_3, output=Base Sprite Tile U}
   ]
  }
  MaterialExpressionAppendVector_7 {
   class=MaterialExpressionAppendVector
   inputs [
    A {from=MaterialExpressionDynamicParameter_4, output=Radial Mask Tile U}
    B {from=MaterialExpressionDynamicParameter_4, output=Radial Mask Tile U}
   ]
  }
  MaterialExpressionAppendVector_9 {
   class=MaterialExpressionAppendVector
   inputs [
    A {from=MaterialExpressionLinearInterpolate_7, output=""}
    B {from=MaterialExpressionLinearInterpolate_8, output=""}
   ]
  }
  MaterialExpressionArctangent2Fast_0 {
   class=MaterialExpressionArctangent2Fast
   inputs [
    Y {from=MaterialExpressionMaterialFunctionCall_1, output=G}
    X {from=MaterialExpressionMaterialFunctionCall_1, output=G}
   ]
   properties {}
  }
  MaterialExpressionChannelMaskParameter_0 {
   class=MaterialExpressionChannelMaskParameter
   inputs [None {from=MaterialExpressionNamedRerouteUsage_6, output=""}]
   properties {group=1. Base Sprite, parameterName=Base Sprite Channel, sortPriority=61}
  }
  MaterialExpressionChannelMaskParameter_1 {
   class=MaterialExpressionChannelMaskParameter
   inputs [None {from=MaterialExpressionNamedRerouteUsage_7, output=""}]
   properties {group=2. Second Sprite, parameterName=Second Sprite Channel, sortPriority=71}
  }
  MaterialExpressionChannelMaskParameter_2 {
   class=MaterialExpressionChannelMaskParameter
   inputs [None {from=MaterialExpressionNamedRerouteUsage_8, output=""}]
   properties {group=3. Radial Mask, parameterName=Mask Channel, sortPriority=71}
  }
  MaterialExpressionClamp_0 {
   class=MaterialExpressionClamp
   inputs [
    None {from=MaterialExpressionMultiply_19, output=""}
    Min {from="", output=""}
    Max {from="", output=""}
   ]
   properties {}
  }
  MaterialExpressionComponentMask_0 {class=MaterialExpressionComponentMask, inputs [None {from=MaterialExpressionNamedRerouteUsage_4}], properties {b=true, g=true, r=true}}
  MaterialExpressionComponentMask_1 {class=MaterialExpressionComponentMask, inputs [None {from=MaterialExpressionNamedRerouteUsage_5}], properties {b=true, g=true, r=true}}
  MaterialExpressionComponentMask_2 {class=MaterialExpressionComponentMask, inputs [None {from=MaterialExpressionVectorParameter_10, output=RGB}], properties {g=true, r=true}}
  MaterialExpressionComponentMask_4 {class=MaterialExpressionComponentMask, inputs [None {from=MaterialExpressionTextureCoordinate_2}], properties {r=true}}
  MaterialExpressionComponentMask_5 {class=MaterialExpressionComponentMask, inputs [None {from=MaterialExpressionTextureCoordinate_2}], properties {g=true}}
  MaterialExpressionConstant_0 {class=MaterialExpressionConstant, properties {r=2}}
  MaterialExpressionConstant_2 {class=MaterialExpressionConstant, properties {r=-1}}
  MaterialExpressionConstant_3 {class=MaterialExpressionConstant, properties {r=1}}
  MaterialExpressionConstant_6 {class=MaterialExpressionConstant, properties {r=-1}}
  MaterialExpressionConstant_7 {class=MaterialExpressionConstant, properties {r=1}}
  MaterialExpressionConvert_0 {
   class=MaterialExpressionConvert
   inputs [
    None {from=MaterialExpressionVectorParameter_10, output=Speed X}
    None {from=MaterialExpressionVectorParameter_10, output=Speed X}
   ]
   properties {
    nodeName=Make Vector2
    convertInputs [
     {
      type=Scalar
      expressionInput {
       expression=&self:MaterialExpressionVectorParameter_10
       inputName=None
       mask=1
       maskA=0
       maskB=1
       maskG=0
       maskR=0
       outputIndex=3
      }
     }
     {
      type=Scalar
      expressionInput {
       expression=&self:MaterialExpressionVectorParameter_10
       inputName=None
       mask=1
       maskA=1
       maskB=0
       maskG=0
       maskR=0
       outputIndex=4
      }
     }
    ]
    convertOutputs [{type=Vector2}]
   }
  }
  MaterialExpressionDDX_0 {class=MaterialExpressionDDX, inputs [Value {from=MaterialExpressionTextureCoordinate_1, output=""}], properties {}}
  MaterialExpressionDDY_0 {class=MaterialExpressionDDY, inputs [Value {from=MaterialExpressionTextureCoordinate_1, output=""}], properties {}}
  MaterialExpressionDepthFade_1 {
   class=MaterialExpressionDepthFade
   inputs [
    Opacity {from=MaterialExpressionReroute_10, output=""}
    FadeDistance {from=MaterialExpressionDynamicParameter_5, output=Depth Fade}
   ]
   properties {}
  }
  MaterialExpressionDivide_0 {
   class=MaterialExpressionDivide
   inputs [
    A {from=MaterialExpressionAdd_2}
    B {from=MaterialExpressionReroute_3}
   ]
  }
  MaterialExpressionDivide_1 {
   class=MaterialExpressionDivide
   inputs [
    A {from=MaterialExpressionSubtract_0}
    B {from=MaterialExpressionSubtract_2}
   ]
  }
  MaterialExpressionDynamicParameter_0 {
   class=MaterialExpressionDynamicParameter
   inputs []
   properties {parameterIndex=3, defaultValue=<rgba 1 1 1 -1>, paramNames [Depth Fade, Twist Center, Twist Edge, Dissolve]}
  }
  MaterialExpressionDynamicParameter_1 {
   class=MaterialExpressionDynamicParameter
   inputs []
   properties {parameterIndex=3, defaultValue=<rgba 1 1 1 -1>, paramNames [Depth Fade, Twist Center, Twist Edge, Dissolve]}
  }
  MaterialExpressionDynamicParameter_3 {
   class=MaterialExpressionDynamicParameter
   inputs []
   properties {parameterIndex=2, paramNames [Base Sprite Tile U, Base Sprite Tile V, Radial Mask Tile U, Radial Mask Tile V]}
  }
  MaterialExpressionDynamicParameter_4 {
   class=MaterialExpressionDynamicParameter
   inputs []
   properties {parameterIndex=2, paramNames [Base Sprite Tile U, Base Sprite Tile V, Radial Mask Tile U, Radial Mask Tile V]}
  }
  MaterialExpressionDynamicParameter_5 {
   class=MaterialExpressionDynamicParameter
   inputs []
   properties {parameterIndex=3, defaultValue=<rgba 1 1 1 -1>, paramNames [Depth Fade, Twist Center, Twist Edge, Dissolve]}
  }
  MaterialExpressionDynamicParameter_6 {
   class=MaterialExpressionDynamicParameter
   inputs []
   properties {defaultValue=<rgba 1 1 2 0>, paramNames [Sprite Extrusion Speed, Sprite Spin Speed, Radial Mask Strength, Radial Mask Power]}
  }
  MaterialExpressionDynamicParameter_7 {
   class=MaterialExpressionDynamicParameter
   inputs []
   properties {parameterIndex=1, defaultValue=<rgba 1 1 1 2>, paramNames [Sprite Offset U, Sprite Offset V, Mask Offset U, Mask Offset V]}
  }
  MaterialExpressionDynamicParameter_8 {
   class=MaterialExpressionDynamicParameter
   inputs []
   properties {defaultValue=<rgba 1 1 2 0>, paramNames [Sprite Extrusion Speed, Sprite Spin Speed, Radial Mask Strength, Radial Mask Power]}
  }
  MaterialExpressionFeatureLevelSwitch_0 {
   class=MaterialExpressionFeatureLevelSwitch
   inputs [
    Default {from=MaterialExpressionReroute_12, output=""}
    ES2 {from=MaterialExpressionReroute_12, output=""}
    ES3_1 {from=MaterialExpressionReroute_12, output=""}
    SM4_REMOVED {from=MaterialExpressionReroute_12, output=""}
    SM5 {from=MaterialExpressionShadingPathSwitch_0, output=""}
    SM6 {from=MaterialExpressionShadingPathSwitch_0, output=""}
   ]
   properties {}
  }
  MaterialExpressionIf_0 {
   class=MaterialExpressionIf
   inputs [
    A {from=MaterialExpressionScalarParameter_19, output=""}
    B {from="", output=""}
    "A > B" {from=MaterialExpressionNamedRerouteUsage_11, output=""}
    "A == B" {from=MaterialExpressionNamedRerouteUsage_1, output=""}
    "A < B" {from=MaterialExpressionNamedRerouteUsage_1, output=""}
   ]
   properties {constB=0.5}
  }
  MaterialExpressionIf_1 {
   class=MaterialExpressionIf
   inputs [
    A {from=MaterialExpressionScalarParameter_20, output=""}
    B {from="", output=""}
    "A > B" {from=MaterialExpressionSaturate_4, output=""}
    "A == B" {from=MaterialExpressionMultiply_1, output=""}
    "A < B" {from=MaterialExpressionMultiply_1, output=""}
   ]
   properties {constB=0.5}
  }
  MaterialExpressionLinearInterpolate_0 {
   class=MaterialExpressionLinearInterpolate
   inputs [
    A {from=MaterialExpressionReroute_5}
    B {from=MaterialExpressionMaterialFunctionCall_2, output=YX}
    Alpha {from=MaterialExpressionScalarParameter_1}
   ]
  }
  MaterialExpressionLinearInterpolate_1 {
   class=MaterialExpressionLinearInterpolate
   inputs [
    A {from=MaterialExpressionConstant_3}
    B {from=MaterialExpressionConstant_2}
    Alpha {from=MaterialExpressionScalarParameter_3}
   ]
  }
  MaterialExpressionLinearInterpolate_10 {
   class=MaterialExpressionLinearInterpolate
   inputs [
    A {from=""}
    B {from=MaterialExpressionClamp_0}
    Alpha {from=MaterialExpressionScalarParameter_15}
   ]
   properties {constA=1}
  }
  MaterialExpressionLinearInterpolate_2 {
   class=MaterialExpressionLinearInterpolate
   inputs [
    A {from=MaterialExpressionConstant_3}
    B {from=MaterialExpressionConstant_2}
    Alpha {from=MaterialExpressionScalarParameter_2}
   ]
  }
  MaterialExpressionLinearInterpolate_3 {
   class=MaterialExpressionLinearInterpolate
   inputs [
    A {from=MaterialExpressionChannelMaskParameter_0, output=R}
    B {from=MaterialExpressionAdd_4}
    Alpha {from=MaterialExpressionScalarParameter_0}
   ]
  }
  MaterialExpressionLinearInterpolate_4 {
   class=MaterialExpressionLinearInterpolate
   inputs [
    A {from=MaterialExpressionDynamicParameter_0, output=Twist Center}
    B {from=MaterialExpressionDynamicParameter_0, output=Twist Center}
    Alpha {from=MaterialExpressionReroute_2}
   ]
   properties {constB=0}
  }
  MaterialExpressionLinearInterpolate_5 {
   class=MaterialExpressionLinearInterpolate
   inputs [
    A {from=MaterialExpressionComponentMask_0}
    B {from=MaterialExpressionAdd_6}
    Alpha {from=MaterialExpressionScalarParameter_12}
   ]
  }
  MaterialExpressionLinearInterpolate_6 {
   class=MaterialExpressionLinearInterpolate
   inputs [
    A {from=MaterialExpressionReroute_0}
    B {from=MaterialExpressionMaterialFunctionCall_6, output=YX}
    Alpha {from=MaterialExpressionScalarParameter_7}
   ]
  }
  MaterialExpressionLinearInterpolate_7 {
   class=MaterialExpressionLinearInterpolate
   inputs [
    A {from=MaterialExpressionConstant_7}
    B {from=MaterialExpressionConstant_6}
    Alpha {from=MaterialExpressionScalarParameter_9}
   ]
  }
  MaterialExpressionLinearInterpolate_8 {
   class=MaterialExpressionLinearInterpolate
   inputs [
    A {from=MaterialExpressionConstant_7}
    B {from=MaterialExpressionConstant_6}
    Alpha {from=MaterialExpressionScalarParameter_8}
   ]
  }
  MaterialExpressionLinearInterpolate_9 {
   class=MaterialExpressionLinearInterpolate
   inputs [
    A {from=MaterialExpressionLinearInterpolate_3}
    B {from=MaterialExpressionNamedRerouteUsage_9}
    Alpha {from=MaterialExpressionScalarParameter_14}
   ]
  }
  MaterialExpressionMaterialFunctionCall_0 {
   class=MaterialExpressionMaterialFunctionCall
   inputs [Emissive {from=MaterialExpressionNamedRerouteUsage_3}]
   properties {materialFunction=&MF_FX_EmissiveAdjustments}
  }
  MaterialExpressionMaterialFunctionCall_1 {
   class=MaterialExpressionMaterialFunctionCall
   inputs [Float2 {from=MaterialExpressionSubtract_1}]
   properties {materialFunction=&BreakOutFloat2Components}
  }
  MaterialExpressionMaterialFunctionCall_2 {
   class=MaterialExpressionMaterialFunctionCall
   inputs [
    XYZ {from=""}
    XY {from=MaterialExpressionReroute_5}
   ]
   properties {materialFunction=&Swizzle}
  }
  MaterialExpressionMaterialFunctionCall_3 {
   class=MaterialExpressionMaterialFunctionCall
   inputs [Multiplier {from=""}]
   properties {materialFunction=&Pi}
  }
  MaterialExpressionMaterialFunctionCall_4 {
   class=MaterialExpressionMaterialFunctionCall
   inputs [Multiplier {from=MaterialExpressionConstant_0}]
   properties {materialFunction=&Pi}
  }
  MaterialExpressionMaterialFunctionCall_5 {
   class=MaterialExpressionMaterialFunctionCall
   inputs [
    Vector 3 {from=""}
    Vector 2 {from=MaterialExpressionSubtract_1}
   ]
   properties {materialFunction=/Engine/Functions/Engine_MaterialFunctions02/Utility/VectorLength}
  }
  MaterialExpressionMaterialFunctionCall_6 {
   class=MaterialExpressionMaterialFunctionCall
   inputs [
    XYZ {from=""}
    XY {from=MaterialExpressionReroute_0}
   ]
   properties {materialFunction=&Swizzle}
  }
  MaterialExpressionMaterialFunctionCall_7 {
   class=MaterialExpressionMaterialFunctionCall
   inputs [
    Fade Length {from=MaterialExpressionScalarParameter_6}
    Fade Offset {from=MaterialExpressionScalarParameter_10}
    For Vertex Shader {from=""}
   ]
   properties {materialFunction=&CameraDepthFade}
  }
  MaterialExpressionMaterialFunctionCall_8 {
   class=MaterialExpressionMaterialFunctionCall
   inputs [
    Alpha {from=MaterialExpressionIf_1}
    Cutoff {from=MaterialExpressionDynamicParameter_1, output=Dissolve}
    Soft Transition {from=MaterialExpressionStaticBool_0}
    Transition Range {from=MaterialExpressionScalarParameter_17}
    Normalized Alpha Range {from=""}
   ]
   properties {materialFunction=/DealershipSharedLibrary/FX/Materials/MaterialFunctions/MF_AlphaThreshold_Smoothstep}
  }
  MaterialExpressionMin_0 {
   class=MaterialExpressionMin
   inputs [
    A {from=MaterialExpressionSaturate_0, output=""}
    B {from=MaterialExpressionSaturate_1, output=""}
   ]
   properties {}
  }
  MaterialExpressionMultiply_0 {
   class=MaterialExpressionMultiply
   inputs [
    A {from=MaterialExpressionMultiply_15}
    B {from=MaterialExpressionAppendVector_6}
   ]
  }
  MaterialExpressionMultiply_1 {
   class=MaterialExpressionMultiply
   inputs [
    A {from=MaterialExpressionLinearInterpolate_9}
    B {from=MaterialExpressionSaturate_4}
   ]
  }
  MaterialExpressionMultiply_11 {
   class=MaterialExpressionMultiply
   inputs [
    A {from=MaterialExpressionChannelMaskParameter_2, output=R}
    B {from=MaterialExpressionDynamicParameter_8, output=Radial Mask Strength}
   ]
  }
  MaterialExpressionMultiply_12 {
   class=MaterialExpressionMultiply
   inputs [
    A {from=MaterialExpressionTextureSampleParameter2D_1, output=RGBA}
    B {from=MaterialExpressionLinearInterpolate_10}
   ]
  }
  MaterialExpressionMultiply_14 {
   class=MaterialExpressionMultiply
   inputs [
    A {from=MaterialExpressionMultiply_28}
    B {from=MaterialExpressionParticleColor_0, output=A}
   ]
  }
  MaterialExpressionMultiply_15 {
   class=MaterialExpressionMultiply
   inputs [
    A {from=MaterialExpressionLinearInterpolate_0}
    B {from=MaterialExpressionAppendVector_1}
   ]
  }
  MaterialExpressionMultiply_17 {
   class=MaterialExpressionMultiply
   inputs [
    A {from=MaterialExpressionSine_1}
    B {from=MaterialExpressionSine_2}
   ]
  }
  MaterialExpressionMultiply_18 {
   class=MaterialExpressionMultiply
   inputs [
    A {from=MaterialExpressionMultiply_17}
    B {from=MaterialExpressionMultiply_17}
   ]
  }
  MaterialExpressionMultiply_19 {
   class=MaterialExpressionMultiply
   inputs [
    A {from=MaterialExpressionMultiply_18}
    B {from=""}
   ]
   properties {constB=2}
  }
  MaterialExpressionMultiply_2 {
   class=MaterialExpressionMultiply
   inputs [
    A {from=MaterialExpressionMultiply_4}
    B {from=MaterialExpressionAppendVector_7}
   ]
  }
  MaterialExpressionMultiply_28 {
   class=MaterialExpressionMultiply
   inputs [
    A {from=MaterialExpressionReroute_12}
    B {from=MaterialExpressionFeatureLevelSwitch_0}
   ]
  }
  MaterialExpressionMultiply_3 {
   class=MaterialExpressionMultiply
   inputs [
    A {from=MaterialExpressionSquareRoot_0}
    B {from=MaterialExpressionScalarParameter_5}
   ]
  }
  MaterialExpressionMultiply_4 {
   class=MaterialExpressionMultiply
   inputs [
    A {from=MaterialExpressionLinearInterpolate_6}
    B {from=MaterialExpressionAppendVector_9}
   ]
  }
  MaterialExpressionMultiply_5 {
   class=MaterialExpressionMultiply
   inputs [
    A {from=MaterialExpressionMultiply_14}
    B {from=MaterialExpressionMaterialFunctionCall_7, output=Result}
   ]
  }
  MaterialExpressionMultiply_6 {
   class=MaterialExpressionMultiply
   inputs [
    A {from=MaterialExpressionReroute_18}
    B {from=MaterialExpressionLinearInterpolate_5}
   ]
  }
  MaterialExpressionMultiply_7 {
   class=MaterialExpressionMultiply
   inputs [
    A {from=MaterialExpressionTextureCoordinate_5}
    B {from=MaterialExpressionComponentMask_2}
   ]
  }
  MaterialExpressionMultiply_9 {
   class=MaterialExpressionMultiply
   inputs [
    A {from=MaterialExpressionMaterialFunctionCall_8, output=OpacityMask}
    B {from=MaterialExpressionIf_1}
   ]
  }
  MaterialExpressionNamedRerouteDeclaration_0 {
   class=MaterialExpressionNamedRerouteDeclaration
   inputs [None {from=MaterialExpressionMultiply_0}]
   properties {name=Sprite UVs, nodeColor=<rgba 1 0.0785 0 1>}
  }
  MaterialExpressionNamedRerouteDeclaration_1 {
   class=MaterialExpressionNamedRerouteDeclaration
   inputs [None {from=MaterialExpressionMultiply_2}]
   properties {name=Mask UVs, nodeColor=<rgba 1 0.0972 0 1>}
  }
  MaterialExpressionNamedRerouteDeclaration_2 {
   class=MaterialExpressionNamedRerouteDeclaration
   inputs [None {from=MaterialExpressionMultiply_5}]
   properties {name=Opacity, nodeColor=<rgba 0 1 0.7188 1>}
  }
  MaterialExpressionNamedRerouteDeclaration_3 {
   class=MaterialExpressionNamedRerouteDeclaration
   inputs [None {from=MaterialExpressionStaticSwitchParameter_0}]
   properties {name=Emissive, nodeColor=<rgba 1 0.3984 0 1>}
  }
  MaterialExpressionNamedRerouteDeclaration_4 {
   class=MaterialExpressionNamedRerouteDeclaration
   inputs [None {from=MaterialExpressionTextureSampleParameter2D_0, output=RGB}]
   properties {name=Base Sprite RGBA, nodeColor=<rgba 1 1 1 1>}
  }
  MaterialExpressionNamedRerouteDeclaration_5 {
   class=MaterialExpressionNamedRerouteDeclaration
   inputs [None {from=MaterialExpressionTextureSampleParameter2D_2, output=RGBA}]
   properties {name=Secoundary Sprite RGBA, nodeColor=<rgba 1 1 1 1>}
  }
  MaterialExpressionNamedRerouteDeclaration_6 {
   class=MaterialExpressionNamedRerouteDeclaration
   inputs [None {from=MaterialExpressionMultiply_12}]
   properties {name=Radial Mask RGBA, nodeColor=<rgba 1 1 1 1>}
  }
  MaterialExpressionNamedRerouteDeclaration_7 {
   class=MaterialExpressionNamedRerouteDeclaration
   inputs [None {from=MaterialExpressionSaturate_2}]
   properties {name=Levels Adjustment, nodeColor=<rgba 0.8272 0.0405 1 1>}
  }
  MaterialExpressionNamedRerouteDeclaration_8 {
   class=MaterialExpressionNamedRerouteDeclaration
   inputs [None {from=MaterialExpressionPanner_3}]
   properties {name=Mask_Simple_UVS, nodeColor=<rgba 1 0.071 0.0185 1>}
  }
  MaterialExpressionNamedRerouteUsage_0 {class=MaterialExpressionNamedRerouteUsage}
  MaterialExpressionNamedRerouteUsage_1 {class=MaterialExpressionNamedRerouteUsage}
  MaterialExpressionNamedRerouteUsage_11 {class=MaterialExpressionNamedRerouteUsage}
  MaterialExpressionNamedRerouteUsage_2 {class=MaterialExpressionNamedRerouteUsage}
  MaterialExpressionNamedRerouteUsage_3 {class=MaterialExpressionNamedRerouteUsage}
  MaterialExpressionNamedRerouteUsage_4 {class=MaterialExpressionNamedRerouteUsage}
  MaterialExpressionNamedRerouteUsage_5 {class=MaterialExpressionNamedRerouteUsage}
  MaterialExpressionNamedRerouteUsage_6 {class=MaterialExpressionNamedRerouteUsage}
  MaterialExpressionNamedRerouteUsage_7 {class=MaterialExpressionNamedRerouteUsage}
  MaterialExpressionNamedRerouteUsage_8 {class=MaterialExpressionNamedRerouteUsage}
  MaterialExpressionNamedRerouteUsage_9 {class=MaterialExpressionNamedRerouteUsage}
  MaterialExpressionPanner_3 {
   class=MaterialExpressionPanner
   inputs [
    Coordinate {from=MaterialExpressionMultiply_7}
    Time {from=""}
    Speed {from=MaterialExpressionConvert_0}
   ]
  }
  MaterialExpressionParticleColor_0 {class=MaterialExpressionParticleColor, inputs [], properties {}}
  MaterialExpressionParticleColor_1 {class=MaterialExpressionParticleColor, inputs [], properties {}}
  MaterialExpressionPower_0 {
   class=MaterialExpressionPower
   inputs [
    Base {from=MaterialExpressionMultiply_3}
    Exp {from=MaterialExpressionScalarParameter_4}
   ]
  }
  MaterialExpressionPower_4 {
   class=MaterialExpressionPower
   inputs [
    Base {from=MaterialExpressionMultiply_11}
    Exp {from=MaterialExpressionDynamicParameter_8, output=Radial Mask Power}
   ]
  }
  MaterialExpressionReroute_0 {class=MaterialExpressionReroute, inputs [None {from=MaterialExpressionAdd_0}]}
  MaterialExpressionReroute_1 {class=MaterialExpressionReroute, inputs [None {from=MaterialExpressionPower_0}]}
  MaterialExpressionReroute_10 {class=MaterialExpressionReroute, inputs [None {from=MaterialExpressionMultiply_9}]}
  MaterialExpressionReroute_11 {class=MaterialExpressionReroute, inputs [None {from=MaterialExpressionReroute_10}]}
  MaterialExpressionReroute_12 {class=MaterialExpressionReroute, inputs [None {from=MaterialExpressionReroute_11}]}
  MaterialExpressionReroute_15 {class=MaterialExpressionReroute, inputs [None {from=MaterialExpressionPower_0}]}
  MaterialExpressionReroute_18 {class=MaterialExpressionReroute, inputs [None {from=MaterialExpressionParticleColor_1, output=RGB}]}
  MaterialExpressionReroute_19 {class=MaterialExpressionReroute, inputs [None {from=MaterialExpressionReroute_20}]}
  MaterialExpressionReroute_2 {class=MaterialExpressionReroute, inputs [None {from=MaterialExpressionReroute_1}]}
  MaterialExpressionReroute_20 {class=MaterialExpressionReroute, inputs [None {from=MaterialExpressionReroute_18}]}
  MaterialExpressionReroute_3 {class=MaterialExpressionReroute, inputs [None {from=MaterialExpressionMaterialFunctionCall_4, output=Result}]}
  MaterialExpressionReroute_4 {class=MaterialExpressionReroute, inputs [None {from=MaterialExpressionAppendVector_2}]}
  MaterialExpressionReroute_5 {class=MaterialExpressionReroute, inputs [None {from=MaterialExpressionAdd_1}]}
  MaterialExpressionReroute_6 {class=MaterialExpressionReroute, inputs [None {from=MaterialExpressionMultiply_5}]}
  MaterialExpressionReroute_7 {class=MaterialExpressionReroute, inputs [None {from=MaterialExpressionLinearInterpolate_3}]}
  MaterialExpressionReroute_8 {class=MaterialExpressionReroute, inputs [None {from=MaterialExpressionMaterialFunctionCall_3, output=Result}]}
  MaterialExpressionSaturate_0 {class=MaterialExpressionSaturate, inputs [None {from=MaterialExpressionScalarParameter_11}]}
  MaterialExpressionSaturate_1 {class=MaterialExpressionSaturate, inputs [None {from=MaterialExpressionScalarParameter_13}]}
  MaterialExpressionSaturate_2 {class=MaterialExpressionSaturate, inputs [None {from=MaterialExpressionDivide_1}]}
  MaterialExpressionSaturate_3 {class=MaterialExpressionSaturate, inputs [None {from=MaterialExpressionMultiply_9}]}
  MaterialExpressionSaturate_4 {class=MaterialExpressionSaturate, inputs [None {from=MaterialExpressionPower_4}]}
  MaterialExpressionScalarParameter_0 {class=MaterialExpressionScalarParameter, properties {group=2. Second Sprite, parameterName=Add Secondary Sprite?, sortPriority=69}}
  MaterialExpressionScalarParameter_1 {class=MaterialExpressionScalarParameter, properties {group=1.1 Base Sprite Rotation, parameterName=Base Sprite - Rotate UVs, sortPriority=0}}
  MaterialExpressionScalarParameter_10 {class=MaterialExpressionScalarParameter, properties {group=0. General Settings, parameterName=Camera Fade - Offset}}
  MaterialExpressionScalarParameter_11 {class=MaterialExpressionScalarParameter, properties {parameterName=Levels - Min}}
  MaterialExpressionScalarParameter_12 {class=MaterialExpressionScalarParameter, properties {group=2. Second Sprite, parameterName=Add Secondary Sprite?, sortPriority=69}}
  MaterialExpressionScalarParameter_13 {class=MaterialExpressionScalarParameter, properties {defaultValue=1, parameterName=Levels - Max}}
  MaterialExpressionScalarParameter_14 {class=MaterialExpressionScalarParameter, properties {parameterName=Base Sprite - Adjust Levels}}
  MaterialExpressionScalarParameter_15 {class=MaterialExpressionScalarParameter, properties {parameterName="Opacity - Use Edge Opacity Mask (0-1)"}}
  MaterialExpressionScalarParameter_17 {class=MaterialExpressionScalarParameter, properties {defaultValue=1, parameterName=Disslve - Transition Softness}}
  MaterialExpressionScalarParameter_19 {class=MaterialExpressionScalarParameter, properties {parameterName=Radial_Maks_UVs_Selector}}
  MaterialExpressionScalarParameter_2 {
   class=MaterialExpressionScalarParameter
   properties {group=1.1 Base Sprite Rotation, parameterName=Base Sprite - Flip UVs Vertically, sortPriority=2}
  }
  MaterialExpressionScalarParameter_20 {class=MaterialExpressionScalarParameter, properties {parameterName=Use_Simple_Mask}}
  MaterialExpressionScalarParameter_3 {
   class=MaterialExpressionScalarParameter
   properties {group=1.1 Base Sprite Rotation, parameterName=Base Sprite - Flip UVs Horizontal, sortPriority=1}
  }
  MaterialExpressionScalarParameter_4 {
   class=MaterialExpressionScalarParameter
   properties {defaultValue=1, group=0. General Settings, parameterName=UV - Radial Power, sliderMax=5}
  }
  MaterialExpressionScalarParameter_5 {class=MaterialExpressionScalarParameter, properties {defaultValue=1, group=0. General Settings, parameterName=UV - Radial Mult, sliderMax=1}}
  MaterialExpressionScalarParameter_6 {class=MaterialExpressionScalarParameter, properties {group=0. General Settings, parameterName=Camera Fade - Length}}
  MaterialExpressionScalarParameter_7 {
   class=MaterialExpressionScalarParameter
   properties {defaultValue=1, group=3.1 Radial Mask Rotation, parameterName=Radial Mask - Rotate UVs, sortPriority=3}
  }
  MaterialExpressionScalarParameter_8 {
   class=MaterialExpressionScalarParameter
   properties {group=3.1 Radial Mask Rotation, parameterName=Radial Mask - Flip UVs Vertically, sortPriority=5}
  }
  MaterialExpressionScalarParameter_9 {
   class=MaterialExpressionScalarParameter
   properties {group=3.1 Radial Mask Rotation, parameterName=Radial Mask - Flip UVs Horizontal, sortPriority=4}
  }
  MaterialExpressionShadingPathSwitch_0 {
   class=MaterialExpressionShadingPathSwitch
   inputs [
    Default {from=MaterialExpressionReroute_11, output=""}
    Deferred {from=MaterialExpressionDepthFade_1, output=""}
    Forward {from="", output=""}
    Mobile {from="", output=""}
   ]
   properties {desc="Defered is pc, Forward is switch, mobile is mobile"}
  }
  MaterialExpressionSine_1 {class=MaterialExpressionSine, inputs [None {from=MaterialExpressionComponentMask_4}]}
  MaterialExpressionSine_2 {class=MaterialExpressionSine, inputs [None {from=MaterialExpressionComponentMask_5}]}
  MaterialExpressionSquareRoot_0 {class=MaterialExpressionSquareRoot, inputs [None {from=MaterialExpressionMaterialFunctionCall_5, output=V2 Length}], properties {}}
  MaterialExpressionStaticBool_0 {class=MaterialExpressionStaticBool, properties {value=true}}
  MaterialExpressionStaticSwitchParameter_0 {
   class=MaterialExpressionStaticSwitchParameter
   inputs [
    True {from=MaterialExpressionMultiply_6, output=""}
    False {from=MaterialExpressionReroute_19, output=""}
   ]
   properties {group=0. General Settings, parameterName=Use Sprite Texture Color}
  }
  MaterialExpressionSubtract_0 {
   class=MaterialExpressionSubtract
   inputs [
    A {from=MaterialExpressionReroute_7}
    B {from=MaterialExpressionMin_0}
   ]
  }
  MaterialExpressionSubtract_1 {
   class=MaterialExpressionSubtract
   inputs [
    A {from=MaterialExpressionTextureCoordinate_8}
    B {from=""}
   ]
   properties {constB=0.5}
  }
  MaterialExpressionSubtract_2 {
   class=MaterialExpressionSubtract
   inputs [
    A {from=MaterialExpressionSaturate_1}
    B {from=MaterialExpressionMin_0}
   ]
  }
  MaterialExpressionTextureCoordinate_1 {class=MaterialExpressionTextureCoordinate}
  MaterialExpressionTextureCoordinate_2 {class=MaterialExpressionTextureCoordinate, properties {uTiling=0.5, vTiling=0.5}}
  MaterialExpressionTextureCoordinate_5 {class=MaterialExpressionTextureCoordinate}
  MaterialExpressionTextureCoordinate_8 {class=MaterialExpressionTextureCoordinate}
  MaterialExpressionTextureSampleParameter2D_0 {
   class=MaterialExpressionTextureSampleParameter2D
   inputs [
    UVs {from=MaterialExpressionNamedRerouteUsage_0, output=""}
    "DDX(UVs)" {from=MaterialExpressionDDX_0, output=""}
    "DDY(UVs)" {from=MaterialExpressionDDY_0, output=""}
    Apply View MipBias {from="", output=""}
   ]
   properties {
    constMipValue=0
    group=1. Base Sprite
    mipValueMode=TMVM_Derivative
    parameterName=Base Sprite
    sortPriority=60
    texture=&T_ShieldCaustics
   }
  }
  MaterialExpressionTextureSampleParameter2D_1 {
   class=MaterialExpressionTextureSampleParameter2D
   inputs [
    UVs {from=MaterialExpressionIf_0, output=""}
    "DDX(UVs)" {from=MaterialExpressionDDX_0, output=""}
    "DDY(UVs)" {from=MaterialExpressionDDY_0, output=""}
    Apply View MipBias {from="", output=""}
   ]
   properties {
    constMipValue=0
    group=3. Radial Mask
    mipValueMode=TMVM_Derivative
    parameterName=Radial Mask
    samplerSource=SSM_Wrap_WorldGroupSettings
    samplerType=SAMPLERTYPE_Masks
    sortPriority=70
    texture=/Game/Global/Textures/Default/Masks_and_Gradients/T_VertGradient_MASK
   }
  }
  MaterialExpressionTextureSampleParameter2D_2 {
   class=MaterialExpressionTextureSampleParameter2D
   inputs [
    UVs {from=MaterialExpressionNamedRerouteUsage_0, output=""}
    "DDX(UVs)" {from=MaterialExpressionDDX_0, output=""}
    "DDY(UVs)" {from=MaterialExpressionDDY_0, output=""}
    Apply View MipBias {from="", output=""}
   ]
   properties {
    constMipValue=0
    group=2. Second Sprite
    mipValueMode=TMVM_Derivative
    parameterName=Secondary Sprite
    sortPriority=70
    texture=/Game/Global/Textures/Default/Caustics/T_Caustics_Softer
   }
  }
  MaterialExpressionVectorParameter_10 {
   class=MaterialExpressionVectorParameter
   properties {group=Main Texture, parameterName=MaskTextureUVs, channelNames {a=Speed Y, b=Speed X, g=Tiling Y, r=Tiling X}, defaultValue=<rgba 1 1 0 0>}
  }
 }
 properties {
  bUsedWithNiagaraMeshParticles=true
  bUsedWithNiagaraSprites=true
  blendMode=BLEND_Translucent
  pixelDepthOffsetMode=PDOM_Legacy
  shadingModel=MSM_Unlit
  twoSided=true
  usedShadingModels=Unlit
 }
}
