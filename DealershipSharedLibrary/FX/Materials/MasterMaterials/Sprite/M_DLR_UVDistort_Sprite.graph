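# Translucent master material for Niagara sprite and mesh particles, building distorted, tintable effects from layered textures.
# Scrolling-free noise, scaled and offset by dynamic parameters, distorts base and mask UVs; detail texture optionally adds sine-driven twinkle.
# Scalar If toggles select mask blending, detail, screen-position color, radial or gradient masking, and MF_DLR_Erode dissolve with colored edge.
# Particle color and dynamic parameters drive emissive and opacity, with brightness-varied depth fade skipped at low quality.
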Material /DealershipSharedLibrary/FX/Materials/MasterMaterials/Sprite/M_DLR_UVDistort_Sprite
graph {
 properties {
  bAutomaticallySetUsageInEditor=false
  bUsedWithNiagaraMeshParticles=true
  bUsedWithNiagaraSprites=true
  blendMode=BLEND_Translucent
  translucencyLightingMode=TLM_VolumetricPerVertexDirectional
  usedShadingModels=Default Lit
 }
 nodes [
  Constant()
  n1 = VectorParameter() {parameterName=Base Color Tint, defaultValue=<rgba 1 1 1 1>}
  n2 = ParticleColor()
  LinearInterpolate(A=@n1.RGB, B=Multiply(A=@n1.RGB, B=@n2.RGB), Alpha=ScalarParameter() {defaultValue=1, parameterName="Base Color - Use Particle Color (0-1)"})
  n3 = ScalarParameter() {group=Dissolve, parameterName=Use Dissolve?, sortPriority=4}
  n4 = NamedRerouteUsage()
  n5 = NamedRerouteUsage()
  n6 = VectorParameter() {parameterName=Detail Texture UVs, channelNames {a=Panner Y, b=Panner X, g=Scale Y, r=Scale X}, defaultValue=<rgba 1 1 0 0>}
  n7 = Panner(Coordinate=Multiply(A=TextureCoordinate(), B=AppendVector(A=@n6.Scale X, B=@n6.Scale X)), Speed=AppendVector(A=@n6.Panner X, B=@n6.Panner X))
  n8 = Multiply(A=NamedRerouteUsage(), B=ScalarParameter() {parameterName=DetailDistortMultiplier})
  n9 = TextureSampleParameter2D(UVs=NamedRerouteDeclaration(None=Add(A=@n7, B=@n8)) {name=DetailTextureUVs, nodeColor=<rgba 0 0.7188 1 1>}) {parameterName=Detail Texture, texture=/DealershipSharedLibrary/Assets/Sets/Ghost/Textures/Noise_Generic_01_Pack}
  n10 = ChannelMaskParameter(None=@n9.RGB) {maskChannel=Blue, parameterName=DetailTexture Channel, defaultValue=<rgba 0 0 1 0>}
  n11 = If(A=ScalarParameter() {parameterName=UseOneChannel Or ThreeChannel Twinkle}, "A > B"=NamedRerouteUsage(), "A == B"=@n10.R, "A < B"=@n10.R)
  n12 = ScalarParameter() {defaultValue=1, parameterName=Detail Texture Pow}
  n13 = DynamicParameter() {defaultValue=<rgba 1 1 1 0>, paramNames [Emissive, Exp, Distortion, MaskDistortion]}
  n14 = TextureCoordinate()
  n15 = DynamicParameter() {parameterIndex=2, paramNames [NoiseScale_X, NoiseScale_Y, NoiseOffset_X, NoiseOffset_Y]}
  n16 = AppendVector(A=@n15.NoiseScale_X, B=@n15.NoiseScale_X)
  n17 = TextureSampleParameter2D(UVs=Add(A=Multiply(A=@n14, B=@n16), B=AppendVector(A=@n15.NoiseOffset_X, B=@n15.NoiseOffset_X))) {parameterName=Noise Texture, texture=/Game/Effects/Fort_Effects/Textures/Tile/T_PerlinRGB_01_M}
  n18 = NamedRerouteDeclaration(None=Subtract(A=ComponentMask(None=@n17.RGB) {g=true, r=true}) {constB=0.5}) {name=DistortTexture, nodeColor=<rgba 0.2891 1 0 1>}
  n19 = TextureCoordinate()
  n20 = DynamicParameter() {parameterIndex=1, paramNames [BaseScale_X,BaseScale_Y,BaseOffset_X,BaseOffset_Y]}
  n21 = Multiply(A=@n19, B=AppendVector(A=@n20.BaseScale_X, B=@n20.BaseScale_X))
  n22 = AppendVector(A=@n20.BaseOffset_X, B=@n20.BaseOffset_X)
  n23 = TextureSampleParameter2D(UVs=Add(A=Multiply(A=@n13.Distortion, B=@n18), B=Add(A=@n21, B=@n22))) {group=Base Texture, parameterName=Base Texture, sortPriority=10, texture=/Game/Effects/Fort_Effects/Textures/Tile/T_Inky_Smoke_Tile2}
  n24 = ChannelMaskParameter(None=@n23.RGBA) {group=Base Texture, parameterName=Base Texture Channel, sortPriority=11}
  n25 = Power(
   Base=Multiply(
    A=ScalarParameter() {defaultValue=1, group=Base Texture, parameterName=Base Texture intensity, sortPriority=12}
    B=StaticSwitchParameter(True=@n24.R, False=@n23.RGB) {group=Base Texture, parameterName=Use Texture Channel, sortPriority=10}
   )
   Exp=DynamicParameter() {defaultValue=<rgba 1 1 1 0>, paramNames [Emissive, Exp, Distortion, MaskDistortion]}.Exp
  )
  n26 = VectorParameter() {
   parameterName=Mask Texture UV Control
   channelNames {a=MaskOffset_Y, b=MaskOffset_X, g=MaskScale_Y, r=MaskScale_X}
   defaultValue=<rgba 1 1 0 0>
  }
  n27 = Multiply(A=@n19, B=AppendVector(A=@n26.MaskScale_X, B=@n26.MaskScale_X))
  n28 = AppendVector(A=@n26.MaskOffset_X, B=@n26.MaskOffset_X)
  n29 = TextureSampleParameter2D(UVs=Add(A=Add(A=@n27, B=@n28), B=Multiply(A=@n18, B=@n13.MaskDistortion))) {
   parameterName=Mask Texture
   samplerType=SAMPLERTYPE_LinearColor
   texture=/Game/Effects/Fort_Effects/Emotes/SmartHyena/T_Smoke_Flipbook_Handpainted
  }
  n30 = NamedRerouteDeclaration(None=ChannelMaskParameter(None=@n29.RGBA) {group=Base Color, parameterName=Mask - Texture Channel}.R) {name=MaskTexture, nodeColor=<rgba 1 0.3984 0 1>}
  n31 = Multiply(A=@n25, B=@n30)
  n32 = If(
   A=ScalarParameter() {desc="[0]Multiply[1]Subtract", parameterName=BaseTexture*Mask}
   "A > B"=Subtract(A=@n25, B=@n30)
   "A == B"=@n31
   "A < B"=@n31
  )
  n33 = If(A=ScalarParameter() {defaultValue=1, parameterName=Use Mask Texture?}, "A > B"=@n32, "A == B"=@n32, "A < B"=@n25) {constB=1}
  n34 = Multiply(A=@n33, B=Multiply(A=Power(Base=@n11, Exp=@n12), B=ScalarParameter() {defaultValue=1, parameterName=DetailTextureMultiplier}))
  n35 = If(A=ScalarParameter() {parameterName=Use Detail Texture?}, "A > B"=@n34, "A == B"=@n34, "A < B"=@n33) {constB=1}
  n36 = Multiply(A=DynamicParameter() {defaultValue=<rgba 1 1 1 0>, paramNames [Emissive, Exp, Distortion, MaskDistortion]}.Emissive, B=@n2.RGB)
  n37 = If(
   A=ScalarParameter() {group=Screen Position RGB Texture, parameterName=Use ScreenPos Tex?, sortPriority=4}
   "A > B"=@n5
   "A == B"=@n5
   "A < B"=If(A=ScalarParameter() {parameterName=Emissive UseTexture}, "A > B"=Multiply(A=@n2.RGB, B=@n35), "A == B"=@n36, "A < B"=@n36)
  ) {constB=1}
  MaterialFunctionCall(Emissive=Multiply(A=If(A=@n3, "A > B"=@n4, "A == B"=@n4, "A < B"=@n37) {constB=1}, B=ScalarParameter() {defaultValue=1, parameterName=Global Emissive Gain})) {materialFunction=&MF_FX_EmissiveAdjustments}
  n38 = MaterialFunctionCall(Float2=TextureCoordinate()) {materialFunction=&BreakOutFloat2Components}
  n39 = Multiply(A=Multiply(A=OneMinus(None=@n38.R), B=@n38.R)) {constB=4}
  n40 = ScalarParameter() {defaultValue=1, parameterName=Mask Distance X}
  n41 = LinearInterpolate(B=Saturate(None=Divide(A=@n39, B=@n40)), Alpha=ScalarParameter() {defaultValue=1, parameterName="Veritcal Gradient Mask Amount (0-1)"}) {constA=1}
  n42 = ScalarParameter() {defaultValue=0.45, parameterName="MaskRadius (S)"}
  n43 = DynamicParameter() {parameterIndex=3, defaultValue=<rgba 1 0 0 0>, paramNames [Dissolve Amount, Sphere Mask Radius, MaskOffset_X, MaskOffset_Y]}
  n44 = If(
   A=ScalarParameter() {parameterName=Use SphereMask?}
   "A > B"=MaterialFunctionCall(UVs=TextureCoordinate(), Radius=Add(A=@n42, B=@n43.Sphere Mask Radius)) {materialFunction=&RadialGradientExponential}.RadialGradientExponential
   "A == B"=@n41
   "A < B"=@n41
  )
  n45 = Multiply(A=@n35, B=@n44)
  n46 = Multiply(A=@n45, B=NamedRerouteUsage())
  n47 = If(A=ScalarParameter() {parameterName=Use DetailTexture Mask}, "A > B"=@n44, "A == B"=@n44, "A < B"=@n45) {constB=1}
  n48 = LinearInterpolate(
   A=ScalarParameter() {defaultValue=400, group=Opacity, parameterName=Depth Fade - Min}
   B=ScalarParameter() {defaultValue=50, group=Opacity, parameterName=Depth Fade - Max}
   Alpha=StaticSwitchParameter(True=@n25, False=Desaturation(None=@n25)) {group=Base Texture, parameterName=Use Texture Channel, sortPriority=10}
  ) {constA=400, constB=50}
  Multiply(
   A=Multiply(A=@n2.A, B=If(A=@n3, "A > B"=@n46, "A == B"=@n46, "A < B"=@n47) {constB=1})
   B=QualitySwitch(Default=DepthFade(FadeDistance=@n48) {fadeDistanceDefault=400}, Low=Constant() {r=1})
  )
  n49 = VectorParameter() {group=Dissolve, parameterName=Dissolve Edge Color, sortPriority=20, defaultValue=<rgba 1 1 1 1>}
  n50 = ScalarParameter() {group=Dissolve, parameterName=Dissolve Edge Gain, sortPriority=24}
  n51 = DynamicParameter() {parameterIndex=3, defaultValue=<rgba 1 1 0 0>, paramNames [Dissolve Amount, Param2, MaskOffset_X, MaskOffset_Y]}
  n52 = TextureCoordinate()
  MaterialExpressionVectorParameter_2 = VectorParameter() {
   group=Dissolve
   parameterName=Dissolve UV Controls
   sortPriority=12
   channelNames {a=Scroll V, b=Scroll U, g=Scale V, r=Scale U}
   defaultValue=<rgba 1 1 0 0>
  }
  n53 = Convert(None=@MaterialExpressionVectorParameter_2.RGBA) {
   nodeName=Redirect UV Controls
   convertInputs [{
     type=Vector4
     defaultValue=<rgba 0 0 0 1>
     expressionInput {
      expression=&self:MaterialExpressionVectorParameter_2
      inputName=None
      mask=1
      maskA=1
      maskB=1
      maskG=1
      maskR=1
      outputIndex=5
     }
    }]
   convertOutputs [{type=Vector2}, {type=Vector2}]
  }
  n54 = TextureSampleParameter2D(UVs=Panner(Coordinate=Multiply(A=@n52, B=@n53), Speed=@n53)) {group=Dissolve, parameterName=Dissolve Texture, sortPriority=8, texture=/Game/Effects/Fort_Effects/Textures/Tile/T_GaussianNoise_01_M}
  n55 = MaterialFunctionCall!(
   Erosion=@n51.Dissolve Amount
   Mask=Saturate(None=@n54.RGB)
   Edge Width=ScalarParameter() {group=Dissolve, parameterName=Dissolve Edge Width, sortPriority=28}
   Width=ScalarParameter() {group=Dissolve, parameterName=Dissolve Smoothness, sortPriority=16}
  ) {materialFunction=/DealershipSharedLibrary/FX/Materials/MaterialFunctions/MF_DLR_Erode}
  NamedRerouteDeclaration(None=LinearInterpolate(A=Multiply(A=@n49.RGB, B=@n50), B=@n37, Alpha=@n55.Edge)) {name=Dissolve Edge Out, nodeColor=<rgba 0.261 0.3828 1 0.5>}
  NamedRerouteDeclaration(None=@n55.Result) {name=Dissolve Out, nodeColor=<rgba 1 0.5 1 0.5>}
  NamedRerouteDeclaration(None=Multiply(A=MaterialFunctionCall(), B=ScalarParameter() {defaultValue=1, group=Screen Position RGB Texture, parameterName=Color Gain, sortPriority=12})) {name=Screen Pos Texture, nodeColor=<rgba 0.1 0.9306 1 0.5>}
  NamedRerouteDeclaration(None=@n9.R) {name=DetailTexture_R, nodeColor=<rgba 1 0 0 1>}
  NamedRerouteDeclaration(None=@n9.G) {name=DetailTexture_G, nodeColor=<rgba 0 1 0 1>}
  NamedRerouteDeclaration(None=@n9.B) {name=DetailTexture_B, nodeColor=<rgba 0 0 1 1>}
  n56 = VectorParameter() {parameterName=TwinkleSpeed, defaultValue=<rgba 1 1 1 1>}
  n57 = Time()
  n58 = Multiply(A=@n57, B=@n56.R)
  n59 = Multiply(A=Clamp(None=Sine(None=@n58)) {minDefault=0.3}, B=NamedRerouteUsage())
  n60 = Multiply(A=@n57, B=@n56.G)
  n61 = Multiply(A=Clamp(None=Sine(None=@n60) {period=3}) {minDefault=0.3}, B=NamedRerouteUsage())
  n62 = Multiply(A=@n56.B, B=@n57)
  n63 = Multiply(A=Clamp(None=Sine(None=@n62) {period=6}) {minDefault=0.3}, B=NamedRerouteUsage())
  n64 = LinearInterpolate(
   A=MaterialFunctionCall(R=@n59, G=@n61, B=@n63) {materialFunction=&AppendMany}.RGB
   B=Add(A=Add(A=@n59, B=@n61), B=@n63)
   Alpha=ScalarParameter() {defaultValue=1, parameterName=use_RGB}
  )
  n65 = Sine(None=Multiply(A=@n56.A, B=@n57)) {period=0.5}
  n66 = TextureSampleParameter2D(UVs=NamedRerouteUsage()) {parameterName=Detail Texture_02, texture=/DealershipSharedLibrary/Assets/Sets/FsDisco/Textures/T_DLR_Floor2_E}
  n67 = VectorParameter() {parameterName=Detail Texture_02_Color, defaultValue=<rgba 1 1 1 1>}
  n68 = Add(A=@n64, B=Multiply(A=Clamp(None=@n65) {maxDefault=0.8, minDefault=0.2}, B=Multiply(A=@n66.R, B=@n67.RGB)))
  NamedRerouteDeclaration(None=If(A=ScalarParameter() {defaultValue=1, parameterName=use_detailtexture02}, "A > B"=@n64, "A == B"=@n68, "A < B"=@n68)) {name=Twinkle, nodeColor=<rgba 1 0 0.6562 1>}
  ScalarParameter() {defaultValue=1, desc=Default 1 clamp at 2, group=Screen Position RGB Texture, parameterName=Saturation Adjust Value, sortPriority=10}
  TextureSampleParameter2D(UVs=ScreenPosition().ViewportUV) {
   group=Screen Position RGB Texture
   parameterName=Screen Position RGB Texture
   sortPriority=8
   texture=/DealershipSharedLibrary/FX/Textures/Diffuse/T_VCVFX_CelestialSpark_Noise_D
  }
 ]
}
